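# Master material for the Space Chimp glider, layering the shared glider base shader, weapon skin FX and an optional metal LUT.
# Bounds-normalized pre-skinned positions build time-animated radial ripples, masked to the far end of local Y.
# The ripples push vertices along their normals and modulate normal-map XY in that region.
# World position offset is zeroed on forward and mobile paths, and tech-art accessory overrides finish the output.
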Material /BRCosmetics/Gadgets/Assets/VinderTech_GliderChute/Glider_Space_Chimp_Male/Materials/M_FN_Glider_SpaceChimp_MASTER
graph {
 properties {
  bUseMaterialAttributes=true
  bUsedWithMeshDeformer=true
  bUsedWithSkeletalMesh=true
  pixelDepthOffsetMode=PDOM_Legacy
  usedShadingModels=Default Lit
 }
 nodes [
  n1 = TextureSampleParameter2D() {group=Skin FX, parameterName=SkinFX_Mask, samplerType=SAMPLERTYPE_Masks, texture=&T_BlackMask}
  n2 = MaterialFunctionCall!(
   Pre-Skinned World Normal
   Gmap
   Customization Mask
   IsCharacter
   IsStatsB
   Material Attributes=MaterialFunctionCall!(UV 1 coordinates) {materialFunction=/Game/Gadgets/Assets/VinderTech_GliderChute/Materials/MF_GliderBaseShader}.Result Material Attributes
  ) {materialFunction=/Game/Weapons/WeaponSkins/MaterialFunction/Weapon_SkinFXFunction}
  n3 = StaticSwitchParameter(
   True=MaterialFunctionCall(Mask=AppendVector(A=@n1.RGB, B=@n1.RGB), Materials=@n2.Result) {materialFunction=&MF_MetalsLUT}.Result
   False=@n2.Result
  ) {group=Skin FX, parameterName=UseMetalLUT}
  n4 = PreSkinnedLocalBounds()
  n5 = Divide(A=Subtract(A=PreSkinnedPosition(), B=@n4.Min), B=@n4.Extents)
  n6 = ShaderStageSwitch(PixelShader=VertexInterpolator(VS=@n5).PS, VertexShader=@n5)
  n7 = VectorParameter() {group=PanningEmissive, parameterName=WPO_UV_Center, defaultValue=<rgba 0.5 0.5 0 1>}
  n8 = Multiply(
   A=Subtract(A=ComponentMask(None=@n6) {b=true, r=true}, B=ComponentMask(None=@n7.RGB) {g=true, r=true}) {constB=0.5}
   B=ComponentMask(None=VectorParameter() {group=PanningEmissive, parameterName=WPO_UV_Scale, defaultValue=<rgba 2 1.5 0 1>}.RGB) {g=true, r=true}
  )
  n9 = Multiply(A=@n8, B=@n8)
  n10 = Add(A=ComponentMask(None=@n9) {r=true}, B=ComponentMask(None=@n9) {g=true})
  n11 = Subtract(
   A=Power(Base=SquareRoot(None=@n10), Exp=ScalarParameter() {defaultValue=4, parameterName=RadialGradientPow})
   B=Multiply(A=Time() {bOverride_Period=true, period=60}, B=ScalarParameter() {defaultValue=0.4, parameterName=RippleSpeed})
  )
  n12 = ScalarParameter() {defaultValue=7, parameterName=RippleCount}
  n13 = Distance(A=Frac(None=Multiply(A=@n11, B=@n12) {constB=10}), B=Constant() {r=0.5})
  n14 = Multiply(A=@n13, B=ScalarParameter() {defaultValue=2, parameterName=WPO_Strength})
  n15 = SmoothStep(Min=Constant() {r=0.9}, Max=Constant() {r=1}, Value=ComponentMask(None=@n6) {g=true})
  n16 = MaterialFunctionCall!(InVector=Multiply(A=VertexNormalWS(), B=Multiply(A=@n14, B=@n15))) {materialFunction=/Game/Athena/Environments/Landscape/Foliage/MF_MakePreciseVec3}
  n17 = Constant()
  n18 = GetMaterialAttributes(None=@n3) {attributeGetTypes [#1]}
  n19 = Power(Base=OneMinus(None=@n13), Exp=ScalarParameter() {defaultValue=1.2, parameterName=NormalPow})
  n20 = ScalarParameter() {defaultValue=1, parameterName=Normal_Strength}
  n21 = AppendVector(A=Multiply(A=ComponentMask(None=@n18.Normal) {g=true, r=true}, B=Multiply(A=@n19, B=@n20)), B=ComponentMask(None=@n18.Normal) {b=true})
  MaterialFunctionCall(In=SetMaterialAttributes!(
    MaterialAttributes=@n3
    World Position Offset=ShadingPathSwitch(Default=@n16.Result, Forward=@n17, Mobile=@n17)
    Normal=LinearInterpolate(A=@n18.Normal, B=@n21, Alpha=@n15)
   ) {attributeSetTypes [#2, #1]}) {materialFunction=&MF_TechArtOverrides_accessories}
 ]
}
